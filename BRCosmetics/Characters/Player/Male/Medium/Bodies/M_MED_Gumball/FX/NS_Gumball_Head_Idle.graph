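# Niagara idle effect for the Gumball character's head, filling a bowl with rolling mesh gumballs sized and counted by user parameters.
# The CPU emitter fakes motion with pendulum constraints, its spring tightness driven by an owner-speed timeline.
# The GPU emitter solves ball-to-ball collisions inside a bounding sphere, using a custom smoothed delta time that varies with owner speed and front-end context.
# Both orient balls by pure roll from relative velocity.
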NiagaraSystem /BRCosmetics/Characters/Player/Male/Medium/Bodies/M_MED_Gumball/FX/NS_Gumball_Head_Idle
dependencies {
 usedDataInterfaces [&NiagaraDataInterfaceCurve, &NiagaraDataInterfaceArrayFloat]
 usedDynamicInputs [&MakeCustomFloatFromBool]
 usedModules [
  &SystemState
  &self:SystemSpawnScript.SystemScriptSource.SystemScriptGraph.NiagaraNodeAssignment_0.SetVariables_#1
 ]
}
emitters {
 Gumballs_CPU {
  emitterUpdateScript [
   &EmitterState(
    Inactive Response=::NewEnumerator1
    Life Cycle Mode=::NewEnumerator1
    Loop Behavior=::NewEnumerator1
    Loop Duration Mode=::NewEnumerator1
    UseLoopDelay=false
   )
   &self:ScratchModule_01(Number Of Spheres:NiagaraInt32=$User.GumballCount) {module=ScratchModule_01}
   &SpawnBurst_Instantaneous(Spawn Count=$Emitter.Number Of Spheres, Use Spawn Probability=false)
  ]
  particleSpawnScript [
   &InitializeParticle(
    Color Mode=::NewEnumerator0
    Mass Max=3
    Mass Min=0.5
    Mass Mode=::NewEnumerator3
    Mesh Scale Mode=::NewEnumerator3
    UsePositionOffset=false
    Write Material Random=true
   )
   &InitialMeshOrientation(Use Rotation Vector=false)
   &SubUVAnimation(End Frame=3, SubUV Animation Mode:ENiagara_SubUVLookupMode=::NewEnumerator2)
   &SphereLocation(
    MaskSpawnGroup=false
    Non Uniform Scale=<1 1 -1>
    Sphere Distribution:ENiagaraSphereDistributionMode=::NewEnumerator3
    Sphere Radius=&Subtract_Float(A=$System.BowlRadius, B=$System.GumballSize)
    Uniform Distribution=0.15
    UseNonUniformScale=true
    UseOffset=false
    UseSphereOrientation=false
   )
   /SetVariables(Particles.PendulumRestAxis:Vector3f=$Output.SphereLocation.SphereNormal) {isSetParameters=true}
  ]
  particleUpdateScript [
   &ParticleState(Kill Particles When Lifetime Has Elapsed=false)
   &ScaleMeshSize(Scale Factor=&VectorFromFloat(Value=$System.GumballSize))
   &Timeline(
    Additional Output:ENiagara_TimelineOutput=::NewEnumerator1
    Additional Output Mode:ENiagara_TimelineOutputMode=::NewEnumerator1
    Float Curve=DataInterface() {propertyValues {Curve {keys [<at 0 0 RCIM_Cubic>, <at 0.9683 0>, {arriveTangent=29.9451, interpMode=RCIM_Cubic, leaveTangent=29.9451, time=1, value=0.95}]}}}
    Play=&SetBoolByFloatComparison(A=&VectorLength(Vector=$Engine.Owner.Velocity, Safe Length Calculation=false), B=0.1)
    Play Rate=256
    Rewind Rate=0.25
    Timeline End=10
   )
   &Drag(Drag=0.2, Use Linear Drag=true, Use Rotational Drag=false)
   &GravityForce(Gravity=<0 0 -840>)
   &PendulumSetup(
    Module.Enable Spring Driver=true
    Pendulum Length:NiagaraFloat=&Subtract_Float(A=$System.BowlRadius, B=$System.GumballSize)
    Pendulum Rest Axis:Vector3f=$Particles.PendulumRestAxis
    "Spring Driven Constraint (Tightness)":NiagaraFloat=&Lerp_Float(A=150, B=0, Alpha=$StackContext.Timeline.Float)
   )
   &SolveForcesAndVelocity(Acceleration Limit=50, Clamp Velocity=false, Limit Acceleration=false)
   &PendulumConstraint()
   &CalculateAccurateVelocity() {module=CalculateAccurateVelocity001}
   /Niagara/Modules/Math/PureRoll_Orientation(
    Surface Normal:Vector3f=&NormalizeVector(
     Vector To Normalize=&Subtract_Vector(A=Linked() {linkedVariable {name=Particles.Position, type=NiagaraPosition}}, B=/Niagara/DynamicInputs/Helpers/SimulationPosition())
     Safe Normalize=true
     Safe Fallback Threshold=0
     Safe Fallback Vector=<0 0 0.001>
    )
    Velocity:Vector3f=&Subtract_Vector(A=$Engine.Owner.Velocity, B=$Particles.Velocity)
   )
  ]
  rendererClasses [&NiagaraMeshRendererProperties]
  renderers [{rendererClass=&NiagaraMeshRendererProperties, rendererIndex=0}]
 }
 Gumballs_Physics_GPU {
  simTarget=GPUComputeSim
  emitterUpdateScript [
   &EmitterState(
    Inactive Response=::NewEnumerator1
    Life Cycle Mode=::NewEnumerator1
    Loop Behavior=::NewEnumerator1
    Loop Duration Mode=::NewEnumerator1
    UseLoopDelay=false
   )
   &Timeline(
    Play=&SetBoolByFloatComparison(A=&VectorLength(Vector=$Engine.Owner.Velocity, Safe Length Calculation=false), B=25)
    Play Rate=4
   )
   &InterpolateOverTime(
    Initial Float Value:NiagaraFloat=&FN_FrontEndFloatModifier(Gameplay=0.0333, Front End=0.0167)
    Rate Of Change:NiagaraFloat=&FN_FrontEndFloatModifier(Gameplay=&Lerp_Float(A=1, B=60, Alpha=$Emitter.Timeline.NormalizedTime), Front End=1)
    Target Float Value:NiagaraFloat=$Engine.DeltaTime
   )
   /SetVariables(Emitter.MyDT:NiagaraFloat=&FN_FrontEndFloatModifier(
     Gameplay=/Niagara/DynamicInputs/Math/Min_Float(A:NiagaraFloat=$Emitter.InterpolateOverTime.InterpolateOverTime_Float.Moving Average, B:NiagaraFloat=0.0333)
     Front End=$Emitter.InterpolateOverTime.InterpolateOverTime_Float.Moving Average
    )) {isSetParameters=true}
   /SetVariables(Emitter.MyDT:NiagaraFloat=$Emitter.InterpolateOverTime.InterpolateOverTime_Float.Moving Average) {enabled=false, isSetParameters=true}
   &self:ScratchModule_03(Number Of Spheres:NiagaraInt32=$User.GumballCount) {module=ScratchModule_03}
   &SpawnBurst_Instantaneous(Spawn Count=$Emitter.Number Of Spheres, Use Spawn Probability=false)
  ]
  particleSpawnScript [
   &InitializeParticle(Mass Max=10, Mass Min=1, Mass Mode=::NewEnumerator3, Mesh Scale Mode=::NewEnumerator3, UsePositionOffset=false, Write Material Random=true)
   &InitialMeshOrientation(Use Rotation Vector=false)
   &SubUVAnimation(End Frame=3, SubUV Animation Mode:ENiagara_SubUVLookupMode=::NewEnumerator2)
   &SphereLocation(
    MaskSpawnGroup=false
    Non Uniform Scale=<1 1 -1>
    Sphere Distribution:ENiagaraSphereDistributionMode=::NewEnumerator3
    Sphere Radius=&Subtract_Float(A=$System.BowlRadius, B=$System.GumballSize)
    Uniform Distribution=0.2
    Uniform Endpoint Offset=1
    UseNonUniformScale=true
    UseOffset=false
    UseSphereOrientation=false
   )
  ]
  particleUpdateScript [
   &ParticleState(Kill Particles When Lifetime Has Elapsed=false)
   &ScaleMeshSize(Scale Factor=&VectorFromFloat(Value=$System.GumballSize))
   &InheritVelocity(Inherited Velocity Amount Scale=<0.05 0.05 0.05>, Speed Limit=false, UseSpeedThreshold=false) {enabled=false}
   &Drag(Drag=&RandomRangeFloat(Minimum=1, Maximum=2), Use Linear Drag=true, Use Rotational Drag=false)
   &GravityForce(Gravity=&RandomRangeVector(Minimum=<0 0 -3000>, Maximum=<1 1 -3400>))
   &SolveForcesAndVelocity(Clamp Velocity=false, DeltaTime:NiagaraFloat=$Emitter.MyDT, Limit Acceleration=false)
   &self:Solve_Intraparticle_Collisions_01(
    Bounding Sphere Origin:Vector3f=&SimulationPosition()
    Bounding Sphere Radius:NiagaraFloat=$System.BowlRadius
    Convergence Speed=0.5
    Iteration Count=8
    Manual DT:NiagaraFloat=$Emitter.MyDT
    Number Of Spheres:NiagaraInt32=$Emitter.Number Of Spheres
    Sphere Radius:NiagaraFloat=$System.GumballSize
   ) {module=Solve_Intraparticle_Collisions_01}
   &PureRoll_Orientation(
    Surface Normal:Vector3f=&NormalizeVector(
     Vector To Normalize=&Subtract_Vector(A=Linked() {linkedVariable {name=Particles.Position, type=NiagaraPosition}}, B=&SimulationPosition())
     Safe Normalize=true
     Safe Fallback Threshold=0
     Safe Fallback Vector=<0 0 0.001>
    )
    Velocity:Vector3f=&Subtract_Vector(A=$Engine.Owner.Velocity, B=$Particles.Velocity)
   )
  ]
  rendererClasses [&NiagaraMeshRendererProperties]
  renderers [{rendererClass=&NiagaraMeshRendererProperties, rendererIndex=0}]
 }
}
summary {
 systemName=NS_Gumball_Head_Idle
 emitters [
  {emitterName=Gumballs_CPU, rendererClasses [&NiagaraMeshRendererProperties]}
  {emitterName=Gumballs_Physics_GPU, simTarget=GPUComputeSim, rendererClasses [&NiagaraMeshRendererProperties]}
 ]
 userVariables [
  User.bIsFrontEnd {type=NiagaraBool}
  User.bIsFrontEndPreview {type=NiagaraBool}
  User.BowlRadius {type=NiagaraFloat, defaultValue=14, description=""}
  User.GumballCount {type=NiagaraInt32, defaultValue=18}
  User.GumballMat {
   type=MaterialInterface
   description=""
   defaultValue {
    struct=NiagaraExt_VariableValue_Object
    value {
     object=&/MI_Gumball_Head_Balls
     objectClass=/Script/Engine.MaterialInterface
    }
   }
  }
  User.GumballSize {type=NiagaraFloat, defaultValue=2, description=""}
 ]
}
