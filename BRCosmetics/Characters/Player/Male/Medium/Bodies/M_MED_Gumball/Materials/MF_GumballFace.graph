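# Material function that draws Gumball's animated flipbook face (brows, eyes, irises, mouth) over incoming material attributes.
# Facial pose parameters are rounded or ceiled and chained through lerps to choose atlas cells, while look poses slide the iris.
# Each feature is scaled and offset on UV1 under master controls, with squash and stretch on the mouth.
# The composite drives base color, emissive and opacity, plus mouth normals and masked specular and roughness.
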MaterialFunction /BRCosmetics/Characters/Player/Male/Medium/Bodies/M_MED_Gumball/Materials/MF_GumballFace
graph {nodes [
  n1 = ScalarParameter() {defaultValue=0.6, parameterName=Eye_Y_Bias}
  n2 = Subtract(A=Multiply(A=ScalarParameter() {group=AnimCurves, parameterName=look_down_pose}) {constB=-1}, B=@n1) {constB=0.4}
  n3 = ScalarParameter() {desc=AnimCurves, group=AnimCurves, parameterName=look_up_pose}
  Add(A=@n2, B=@n3)
  AppendVector(A=ScalarParameter() {group=AnimCurves, parameterName=Mouth_X_Frame}, B=ScalarParameter() {group=AnimCurves, parameterName=Mouth_Y_Frame})
  AppendVector(A=ScalarParameter() {group=AnimCurves, parameterName=X_BrowFrame}, B=ScalarParameter() {group=AnimCurves, parameterName=Y_BrowFrame})
  AppendVector(A=ScalarParameter() {group=AnimCurves, parameterName=Eye_X_Frame}, B=ScalarParameter() {group=AnimCurves, parameterName=Eye_Y_Frame})
  n4 = GetMaterialAttributes(None=FunctionInput() {inputName=Material, inputType=FunctionInput_MaterialAttributes}) {attributeGetTypes [
    #1
    #2
    #3
    #4
   ]}
  n5 = Divide(B=AppendVector(A=ScalarParameter() {defaultValue=1, parameterName=X_Slides}, B=ScalarParameter() {defaultValue=4, parameterName=Y_Slides}))
  n6 = ScalarParameter() {parameterName=X_BrowOffset}
  n7 = ScalarParameter() {defaultValue=-0.032, parameterName=Y_BrowOffset}
  n8 = Multiply(A=ScalarParameter() {defaultValue=0.5, parameterName=EyeBrowscale}, B=NamedRerouteUsage())
  n9 = ScalarParameter() {defaultValue=0.32, parameterName=EyeBrowScaleBIas}
  n10 = MaterialFunctionCall!(
   UVs=NamedRerouteUsage()
   Offset=Add(A=NamedRerouteUsage(), B=AppendVector(A=@n6, B=@n7))
   Y_MaskRadius=Constant() {r=0.5}
   Texture Scale=MaterialFunctionCall(X=@n8, Y=Multiply(A=@n8, B=@n9)) {materialFunction=&MakeFloat2}.Result
  ) {materialFunction=/Game/Characters/Player/Male/Medium/Bodies/M_MED_Gumball/Materials/MF_ScaleAndOffsetUVs}
  n11 = NamedRerouteUsage()
  n12 = TextureSample(UVs=Add(A=Multiply(A=@n5, B=@n10.UVs), B=Divide(A=@n11) {constB=4})) {texture=/BRCosmetics/Characters/Player/Male/Medium/Bodies/M_MED_Gumball/Textures/T_Gumball_Brows}
  n13 = Divide(B=AppendVector(A=ScalarParameter() {defaultValue=1, parameterName=X_Slides}, B=ScalarParameter() {defaultValue=4, parameterName=Y_Slides}))
  n14 = ScalarParameter() {parameterName=X_EyesOffset}
  n15 = ScalarParameter() {parameterName=Y_EyesOffset}
  n16 = Multiply(A=ScalarParameter() {defaultValue=0.4, parameterName=Eyescale}, B=NamedRerouteUsage())
  n17 = ScalarParameter() {defaultValue=0.3333, parameterName=EyescaleYBias}
  n18 = MaterialFunctionCall(
   UVs=NamedRerouteUsage()
   Offset=Add(A=NamedRerouteUsage(), B=AppendVector(A=@n14, B=@n15))
   Texture Scale=MaterialFunctionCall(X=@n16, Y=Multiply(A=@n16, B=@n17)) {materialFunction=&MakeFloat2}.Result
  ) {materialFunction=&MF_ScaleAndOffsetUVs}
  n19 = NamedRerouteUsage()
  n20 = TextureSample(UVs=Add(A=Multiply(A=@n13, B=@n18.UVs), B=Divide(A=@n19) {constB=3})) {texture=/BRCosmetics/Characters/Player/Male/Medium/Bodies/M_MED_Gumball/Textures/T_Gumball_Eyes}
  n21 = Multiply(A=@n20.A, B=@n18.0-1 mask)
  n22 = Multiply(A=LinearInterpolate(B=@n12.G, Alpha=@n10.0-1 mask) {constA=1}, B=@n21) {constA=1}
  n23 = LinearInterpolate(A=@n12.R, B=@n20.RGB, Alpha=@n22)
  n24 = NamedRerouteUsage()
  n25 = Add(
   A=NamedRerouteUsage()
   B=AppendVector(A=ScalarParameter() {parameterName=Iris_XPos}, B=ScalarParameter() {defaultValue=0.05, parameterName=Iris_YPos})
  )
  n26 = Add(
   A=ScalarParameter() {desc=AnimCurves, group=AnimCurves, parameterName=look_right_pose}
   B=Multiply(A=ScalarParameter() {desc=AnimCurves, group=AnimCurves, parameterName=look_left_pose}) {constB=-1}
  )
  n27 = ScalarParameter() {defaultValue=-0.05, parameterName=AE_EyeHorizontalThreshold_Min}
  n28 = ScalarParameter() {defaultValue=0.05, parameterName=AE_EyeHorizontalThreshold_Max}
  n29 = MaterialFunctionCall(Input=Clamp(None=@n26) {minDefault=-1}, Input Low=Constant() {r=-1}, Input High=Constant() {r=1}, Target Low=@n27, Target High=@n28) {materialFunction=&RemapValueRange}
  n30 = MaterialFunctionCall(
   Input=Clamp(None=@n2) {minDefault=-1}
   Input Low=Constant() {r=-0.5}
   Input High=Constant() {r=0.5}
   Target Low=ScalarParameter() {defaultValue=-0.07, parameterName=AE_EyeVerticalThreshold_Min}
   Target High=ScalarParameter() {defaultValue=0.05, parameterName=AE_EyeVerticalThreshold_Max}
  ) {materialFunction=&RemapValueRange}
  n31 = Clamp(None=Subtract(A=@n3, B=@n1)) {maxDefault=0.5, minDefault=-0.5}
  n32 = LinearInterpolate(
   A=@n30.Result
   B=MaterialFunctionCall(
    Input=@n31
    Input Low=Constant() {r=-0.5}
    Input High=Constant() {r=0.5}
    Target Low=ScalarParameter() {defaultValue=-0.1, parameterName=AE_EyeVerticalThreshold_Min_Up}
    Target High=ScalarParameter() {defaultValue=-0.05, parameterName=AE_EyeVerticalThreshold_Max_Up}
   ) {materialFunction=&RemapValueRange}.Result
   Alpha=Clamp(None=Ceil(None=@n3))
  )
  n33 = ScalarParameter() {defaultValue=0.38, parameterName=AE_PupilScale}
  n34 = Multiply(A=@n33) {constB=0.5}
  n35 = MaterialFunctionCall(
   UVs=Add(A=Add(A=@n24, B=@n25), B=AppendVector(A=@n29.Result, B=@n32))
   Texture Scale=Multiply(A=AppendVector(A=@n33, B=@n34), B=NamedRerouteUsage())
  ) {materialFunction=&ScaleUVsByCenter}
  n36 = TextureSample(UVs=@n35.UVs) {texture=/BRCosmetics/Characters/Player/Male/Medium/Bodies/M_MED_Gumball/Textures/T_Gumball_Iris}
  n37 = Multiply(A=@n20.R, B=@n36.RGB)
  n38 = Multiply(A=@n36.A, B=@n35.0-1 mask)
  n39 = Multiply(A=@n22, B=@n38)
  n40 = Multiply(A=@n10.0-1 mask, B=@n12.A)
  n41 = Constant() {r=1}
  n42 = AppendVector(A=ScalarParameter() {defaultValue=4, parameterName=Mouth_X_Slides}, B=ScalarParameter() {defaultValue=4, parameterName=Mouth_Y_Slides})
  n43 = NamedRerouteUsage()
  n44 = AppendVector(A=ScalarParameter() {parameterName=Mouth_X_Offset}, B=ScalarParameter() {defaultValue=-0.424, parameterName=Mouth_Y_EyesOffset})
  n45 = Multiply(A=ScalarParameter() {defaultValue=0.35, parameterName=MouthScale}, B=NamedRerouteUsage())
  n46 = Multiply(A=@n45, B=ScalarParameter() {defaultValue=1, parameterName=MouthScale_Y_Bias})
  n47 = Constant() {r=1}
  n48 = ScalarParameter() {defaultValue=0.5, parameterName=StretchMin}
  n49 = ScalarParameter() {parameterName=Stretch}
  n50 = Constant() {r=1}
  n51 = ScalarParameter() {defaultValue=2, parameterName=SquashMax}
  n52 = ScalarParameter() {parameterName=Squash}
  n53 = MaterialFunctionCall(
   UVs=MaterialFunctionCall(
    UVs=NamedRerouteUsage()
    Offset=Add(A=@n43, B=@n44)
    Y_MaskRadius=Constant() {r=0.5}
    Texture Scale=MaterialFunctionCall(X=@n45, Y=@n46) {materialFunction=&MakeFloat2}.Result
   ) {materialFunction=&MF_ScaleAndOffsetUVs}.UVs
   Texture Scale=AppendVector(A=LinearInterpolate(A=@n47, B=@n48, Alpha=@n49), B=LinearInterpolate(A=@n50, B=@n51, Alpha=@n52))
  ) {materialFunction=&ScaleUVsByCenter}
  n54 = Add(A=Multiply(A=Divide(A=@n41, B=@n42), B=@n53.UVs), B=Divide(A=NamedRerouteUsage()) {constB=4})
  n55 = TextureSample(UVs=@n54) {texture=/BRCosmetics/Characters/Player/Male/Medium/Bodies/M_MED_Gumball/Textures/T_Gumball_Mouth}
  n56 = Multiply(A=@n55.A, B=@n53.0-1 mask)
  n57 = LinearInterpolate(A=LinearInterpolate(A=LinearInterpolate(A=@n23, B=@n37, Alpha=@n39) {constA=1}, B=@n12.R, Alpha=@n40), B=@n55.RGB, Alpha=@n56)
  n58 = Constant3Vector() {constant=<rgba 0.5 0.5 1 0>}
  n59 = TextureSample(UVs=@n54) {samplerType=SAMPLERTYPE_Normal, texture=/BRCosmetics/Characters/Player/Male/Medium/Bodies/M_MED_Gumball/Textures/T_Gumball_Mouth_Normals}
  n60 = NamedRerouteUsage()
  n61 = Constant()
  n62 = Multiply(
   A=TextureSample(UVs=@n54) {samplerType=SAMPLERTYPE_Masks, texture=/BRCosmetics/Characters/Player/Male/Medium/Bodies/M_MED_Gumball/Textures/T_Mouth_FX}.RGB
   B=@n53.0-1 mask
  )
  n63 = Max(A=ChannelMaskParameter(None=@n62) {parameterName="InnerMouthChanne;"}.R, B=Max(A=@n40, B=@n21))
  n64 = ChannelMaskParameter(None=@n62) {maskChannel=Green, parameterName=TeethChannel, defaultValue=<rgba 0 1 0 0>}
  n65 = NamedRerouteUsage()
  n66 = Constant() {r=1}
  n67 = ScalarParameter() {defaultValue=0.25, parameterName=FaceEmissiveStrength}
  n68 = Multiply(A=@n12.G, B=@n10.0-1 mask)
  n69 = Max(A=@n40, B=Max(A=Multiply(A=@n68, B=@n21), B=@n56))
  n70 = SetMaterialAttributes!(
   MaterialAttributes=@n4.MaterialAttributes
   BaseColor=@n57
   Opacity=@n69
   Normal=MaterialFunctionCall(BaseNormal=NamedRerouteUsage(), AdditionalNormal=LinearInterpolate(A=@n58, B=@n59.RGB, Alpha=@n55.A)) {materialFunction=&BlendAngleCorrectedNormals}.Result
   Specular=LinearInterpolate(A=LinearInterpolate(A=@n60, B=@n61, Alpha=@n63), B=@n60, Alpha=@n64.R)
   Roughness=LinearInterpolate(A=LinearInterpolate(A=@n65, B=@n66, Alpha=@n63), B=@n65, Alpha=@n64.R)
   EmissiveColor=LinearInterpolate(A=NamedRerouteUsage(), B=Multiply(A=@n57, B=@n67), Alpha=@n69)
   OpacityMask=@n69
  ) {attributeSetTypes [
    #5
    #6
    #1
    #2
    #3
    #4
    #7
   ]}
  BlendMaterialAttributes(A=@n4.MaterialAttributes, B=@n70, Alpha=@n69)
  Ceil(None=@n69)
  Clamp(None=@n26, Min=@n27, Max=@n28) {minDefault=-1}
  n71 = TextureCoordinate()
  n72 = Add(
   A=NamedRerouteUsage()
   B=AppendVector(A=ScalarParameter() {parameterName=Mouth_X_Offset}, B=ScalarParameter() {defaultValue=-0.424, parameterName=Mouth_Y_EyesOffset})
  )
  n73 = Constant() {r=0.5}
  n74 = Multiply(A=ScalarParameter() {defaultValue=0.35, parameterName=MouthScale}, B=NamedRerouteUsage())
  n75 = MaterialFunctionCall(X=@n74, Y=Multiply(A=@n74, B=ScalarParameter() {defaultValue=1, parameterName=MouthScale_Y_Bias})) {materialFunction=&MakeFloat2}
  ComponentMask(None=MaterialFunctionCall!(
    Texture=TextureObject() {texture=&T_Gumball_Mouth}
    UVs=MaterialFunctionCall(UVs=@n71, Offset=@n72, Y_MaskRadius=@n73, Texture Scale=@n75.Result) {materialFunction=&MF_ScaleAndOffsetUVs}.UVs
    SubImages=ScalarParameter() {defaultValue=11, parameterName=SubImages}
    Frame=Constant() {r=6}
   ) {materialFunction=/Game/Weapons/WeaponSkins/MaterialFunction/CustomNumbers_SubUV}.RGB) {b=true, g=true, r=true}
  n76 = PreSkinnedPosition()
  ComponentMask(None=ShaderStageSwitch(PixelShader=VertexInterpolator(VS=@n76).PS, VertexShader=@n76)) {b=true, r=true}
  Constant3Vector() {constant=<rgba 1 1 1 0>}
  Constant() {r=1}
  n77 = Add(
   A=TextureCoordinate() {coordinateIndex=1}
   B=AppendVector(A=ScalarParameter() {parameterName=Coord_offset_X}, B=ScalarParameter() {parameterName=Coord_offset_Y})
  )
  Divide(A=@n77, B=ScalarParameter() {defaultValue=32, parameterName=Scale})
  FunctionOutput(None=@n70)
  MaterialFunctionCall() {materialFunction=&ScaleUVsByCenter}
  Max(A=@n38, B=@n56)
  Multiply(A=@n31)
  Multiply(A=@n12.G)
  n78 = Constant2Vector()
  n79 = Constant2Vector() {g=1}
  n80 = ScalarParameter() {defaultValue=0.5, group=Banana_Poses, parameterName=R_squint_inner_pose}
  n81 = Divide(B=ScalarParameter() {defaultValue=0.5, group=Banana, parameterName=Threshold_squint_inner_pose}) {constA=0.5}
  n82 = Clamp(None=Round(None=Multiply(A=@n80, B=@n81)))
  n83 = ScalarParameter() {defaultValue=0.3, group=Banana, parameterName=Threshold_R_blink_pose}
  n84 = Round(None=Multiply(A=ScalarParameter() {group=Banana_Poses, parameterName=R_blink_pose}, B=Divide(B=@n83) {constA=0.5}))
  NamedRerouteDeclaration(None=LinearInterpolate(A=LinearInterpolate(A=@n78, B=@n79, Alpha=@n82), B=Constant2Vector() {g=2}, Alpha=Clamp(None=@n84))) {name=EyeCurveResult, nodeColor=<rgba 0 1 0.2031 1>}
  n85 = Multiply(
   A=ScalarParameter() {group=Banana_Poses, parameterName=R_brow_up_pose}
   B=Divide(B=ScalarParameter() {defaultValue=0.3, group=Banana, parameterName=Threshold_R_brow_up_pose}) {constA=0.5}
  )
  n86 = LinearInterpolate(A=Constant2Vector() {g=3}, B=Constant2Vector(), Alpha=Clamp(None=Round(None=@n85)))
  n87 = Constant2Vector() {g=2}
  n88 = ScalarParameter() {group=Banana_Poses, parameterName=C_glabella_down_pose}
  n89 = Divide(B=ScalarParameter() {defaultValue=0.2, group=Banana, parameterName=Threshold_Glabella_down}) {constA=0.5}
  n90 = Clamp(None=Round(None=Multiply(A=@n88, B=@n89)))
  n91 = ScalarParameter() {defaultValue=0.5, group=Banana, parameterName=Threshold_Glabella_up}
  n92 = Round(None=Multiply(A=ScalarParameter() {group=Banana_Poses, parameterName=C_glabella_up_pose}, B=Divide(B=@n91) {constA=0.5}))
  NamedRerouteDeclaration(None=LinearInterpolate(A=LinearInterpolate(A=@n86, B=@n87, Alpha=@n90), B=Constant2Vector() {g=1}, Alpha=Clamp(None=@n92))) {name=EyebrowCurves, nodeColor=<rgba 0.1719 1 0 1>}
  n93 = Ceil(None=ScalarParameter() {parameterName=R_upper_lip_raiser_pose})
  n94 = LinearInterpolate(A=Constant2Vector(), B=Constant2Vector() {g=1, r=3}, Alpha=@n93)
  n95 = Constant2Vector() {g=2}
  n96 = Ceil(None=ScalarParameter() {parameterName=R_smile_pose})
  n97 = Ceil(None=ScalarParameter() {parameterName=R_lip_corner_narrow_pose})
  n98 = Ceil(None=ScalarParameter() {parameterName=R_Frown_pose})
  n99 = LinearInterpolate(
   A=LinearInterpolate(A=LinearInterpolate(A=@n94, B=@n95, Alpha=@n96), B=Constant2Vector() {r=3}, Alpha=@n97)
   B=Constant2Vector() {g=1}
   Alpha=@n98
  )
  n100 = Constant2Vector() {g=2, r=1}
  n101 = Ceil(None=ScalarParameter() {parameterName=L_upper_lip_raiser_pose})
  n102 = Constant2Vector() {r=1}
  n103 = Constant2Vector() {g=1, r=1}
  n104 = LinearInterpolate(
   A=LinearInterpolate(A=LinearInterpolate(A=@n102, B=@n103, Alpha=@n93), B=Constant2Vector() {g=1, r=2}, Alpha=@n96)
   B=Constant2Vector() {r=2}
   Alpha=@n97
  )
  n105 = Constant2Vector() {g=2, r=2}
  n106 = ScalarParameter() {parameterName=jaw_open_pose}
  n107 = LinearInterpolate(
   A=LinearInterpolate(A=LinearInterpolate(A=@n99, B=@n100, Alpha=@n101), B=LinearInterpolate(A=@n104, B=@n105, Alpha=@n98), Alpha=Ceil(None=@n106))
   B=Constant2Vector() {g=2, r=3}
   Alpha=ScalarParameter() {parameterName=MaterialCurveA}
  )
  n108 = Constant2Vector() {g=3}
  n109 = ScalarParameter() {parameterName=MaterialCurveB}
  NamedRerouteDeclaration(None=LinearInterpolate(A=LinearInterpolate(A=@n107, B=@n108, Alpha=@n109), B=Constant2Vector() {g=3, r=1}, Alpha=ScalarParameter() {parameterName=MaterialCurveC})) {name=MouthCurves, nodeColor=<rgba 0.9688 1 0 1>}
  NamedRerouteDeclaration(None=ScalarParameter() {defaultValue=1.781, parameterName=Face_MacroScale}) {name=MasterScale, nodeColor=<rgba 0.8281 0 1 1>}
  NamedRerouteDeclaration(None=AppendVector(A=ScalarParameter() {parameterName=FaceOffset_X}, B=ScalarParameter() {parameterName=FaceOffset_Y})) {name=MasterOffset, nodeColor=<rgba 0 0.5078 1 1>}
  NamedRerouteDeclaration(None=@n77) {name=FaceUV, nodeColor=<rgba 0.8516 0 1 1>}
  NamedRerouteDeclaration(None=@n4.Normal) {name=BaseNormal, nodeColor=<rgba 1 0 0.3516 1>}
  NamedRerouteDeclaration(None=@n4.Specular) {name=BaseSpecular, nodeColor=<rgba 0 1 0.8125 1>}
  NamedRerouteDeclaration(None=@n4.Roughness) {name=BaseRoughness, nodeColor=<rgba 0.1953 0 1 1>}
  NamedRerouteDeclaration(None=@n4.EmissiveColor) {name=BaseEmissive, nodeColor=<rgba 0.7578 1 0 1>}
  NamedRerouteUsage()
  ObjectRadius()
  Reroute(None=TextureCoordinate() {coordinateIndex=1})
  Subtract(A=@n30.Result, B=@n1)
  TextureCoordinate() {coordinateIndex=1}
  TextureCoordinate() {coordinateIndex=1}
  TextureCoordinate() {coordinateIndex=1}
  TextureCoordinate()
  TextureCoordinate()
 ]}
