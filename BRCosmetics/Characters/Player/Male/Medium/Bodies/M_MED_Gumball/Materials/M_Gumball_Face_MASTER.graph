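# Translucent master material for the Gumball character's face, rendering a lit glass-like shell with a view-dependent film gradient.
# A curve-atlas gradient, driven by view angle and pre-skinned height, tints the base color, while Fresnel and height shape rim color and opacity.
# A sphere-masked half-vector sun highlight feeds emissive and opacity, and tiled noise perturbs the baked normal.
# MF_GumballFace output blends in through UseFace before MF_TechArtOverrides.
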Material /BRCosmetics/Characters/Player/Male/Medium/Bodies/M_MED_Gumball/Materials/M_Gumball_Face_MASTER
graph {
 properties {
  bEnableResponsiveAA=true
  bUseMaterialAttributes=true
  bUsedWithMeshDeformer=true
  bUsedWithMorphTargets=true
  bUsedWithSkeletalMesh=true
  blendMode=BLEND_Translucent
  pixelDepthOffsetMode=PDOM_Legacy
  shadingModel=MSM_FromMaterialExpression
  translucencyLightingMode=TLM_SurfacePerPixelLighting
  usedShadingModels=Default Lit
 }
 nodes [
  n1 = ScalarParameter() {defaultValue=1, parameterName=Fresnel Exponent}
  n2 = Clamp(None=Multiply(A=Fresnel(ExponentIn=@n1), B=ScalarParameter() {defaultValue=2, parameterName=Glass thickness}))
  n3 = PreSkinnedPosition()
  n4 = ShaderStageSwitch(PixelShader=VertexInterpolator(VS=@n3).PS, VertexShader=@n3)
  n5 = VectorParameter() {parameterName=Offset, defaultValue=<rgba 0 0 -148 1>}
  n6 = ScalarParameter() {defaultValue=32, parameterName=HeadScale}
  n7 = ChannelMaskParameter(None=Divide(A=Add(A=@n4, B=@n5.RGB), B=@n6)) {maskChannel=Blue, defaultValue=<rgba 0 0 1 0>}
  n8 = Max(A=DotProduct(A=CameraVectorWS(), B=PixelNormalWS()), B=Clamp(None=OneMinus(None=@n7.R)))
  n9 = CurveAtlasRowParameter(CurveTime=@n8) {
   atlas=&AthenaCurveAtlas01
   curve=/Game/Characters/Player/Male/Medium/Bodies/M_MED_Gumball/Textures/Gumball_FilmCurve
   defaultValue=311
   parameterName=FilmGradient
  }
  n10 = Multiply(
   A=LinearInterpolate(
    A=ScalarParameter() {parameterName=Fresnel Inner Emissive}
    B=ScalarParameter() {defaultValue=0.85, parameterName=Fresnel Outer Emissive}
    Alpha=@n2
   )
   B=@n9
  )
  n11 = VectorParameter() {parameterName=Base Color}
  n12 = Fresnel(ExponentIn=ScalarParameter() {defaultValue=4, parameterName=RimFresnel Exponent})
  n13 = ScalarParameter() {defaultValue=2, parameterName=Glass thickness}
  n14 = Clamp(None=Power(Base=Multiply(A=@n12, B=@n13)) {constExponent=4})
  n15 = TextureSampleParameter2D() {parameterName=Input S, texture=/Game/Global/Textures/Default/Blanks/T_BlackUserInterface}
  n16 = CollectionParameter() {
   collection=&FortniteMaterialParameters
   parameterName=SunlightDirectionalVector
  }
  n17 = Add(A=CameraVectorWS(), B=ComponentMask(None=@n16) {b=true, g=true, r=true})
  n18 = SphereMask(
   A=DotProduct(A=Normalize(VectorInput=@n17), B=PixelNormalWS())
   B=Constant() {r=1}
   Radius=ScalarParameter() {defaultValue=0.02, parameterName=SpecPower}
   Hardness=ScalarParameter() {defaultValue=1, parameterName=Sharpness}
  )
  n19 = TextureSampleParameter2D() {parameterName=Input FX, texture=&T_BlackUserInterface}
  n20 = ScalarParameter() {parameterName=Add Custom Opacity, sortPriority=10}
  n21 = Divide(A=Add(A=@n4, B=VectorParameter() {parameterName=TransparencyOffset, defaultValue=<rgba 0 0 -148 1>}.RGB), B=@n6)
  n22 = LinearInterpolate(
   A=ScalarParameter() {defaultValue=0.5, parameterName=BottomTransparency}
   B=LinearInterpolate(
    A=ScalarParameter() {parameterName=Fresnel Inner Transparency}
    B=ScalarParameter() {defaultValue=0.85, parameterName=Fresnel Outer Transparency}
    Alpha=@n2
   )
   Alpha=Clamp(None=ChannelMaskParameter(None=@n21) {maskChannel=Blue, defaultValue=<rgba 0 0 1 0>}.R)
  )
  n23 = ScalarParameter() {defaultValue=1, parameterName=RimTransparency}
  n24 = LinearInterpolate(
   A=Add(A=Multiply(A=@n19.B, B=@n20), B=LinearInterpolate(A=@n22, B=@n23, Alpha=@n14))
   B=@n19.B
   Alpha=ScalarParameter() {parameterName=Use Custom Opacity}
  )
  n25 = ScalarParameter() {defaultValue=0.5, parameterName=Spec_Emissive}
  n26 = ObjectBounds()
  n27 = PreSkinnedLocalBounds()
  n28 = Saturate(None=Floor(None=ScalarParameter() {parameterName=Use Pre-Skinned Bounds, sliderMax=1}))
  n29 = ComponentMask(None=Divide(A=PreSkinnedPosition(), B=LinearInterpolate(A=@n26, B=@n27.Half Extents, Alpha=@n28))) {b=true, g=true}
  n30 = ScalarParameter() {defaultValue=0.2, parameterName=WaterTiling}
  n31 = TextureSampleParameter2D(UVs=VertexInterpolator(VS=Multiply(A=@n29, B=@n30)).PS) {
   parameterName=Noise Normal
   samplerSource=SSM_Wrap_WorldGroupSettings
   samplerType=SAMPLERTYPE_Normal
   texture=/Game/Global/Textures/Default/Noise/T_Noise_Normal_01
  }
  n32 = VectorParameter() {parameterName=Normal Intensity}
  n33 = TextureSampleParameter2D() {
   group=Mask Texture
   parameterName=BakedNormal
   samplerSource=SSM_Wrap_WorldGroupSettings
   samplerType=SAMPLERTYPE_Normal
   texture=/Engine/EngineMaterials/FlatNormal
  }
  n34 = Normalize(VectorInput=Add(A=Multiply(A=@n31.RGB, B=@n32.RGB), B=@n33.RGB))
  n35 = MakeMaterialAttributes!(
   BaseColor=LinearInterpolate(A=Add(A=@n10, B=@n11.RGB), B=VectorParameter() {desc=1, parameterName=RimColor, defaultValue=<rgba 1 1 1 1>}.RGB, Alpha=@n14)
   Metallic=LinearInterpolate(A=ScalarParameter() {defaultValue=0.1, parameterName=Metallic}, B=@n15.G, Alpha=ScalarParameter() {parameterName=Use Custom Metallic})
   Specular=LinearInterpolate(A=ScalarParameter() {defaultValue=1, parameterName=Specular}, B=@n15.R, Alpha=ScalarParameter() {parameterName=Use Custom Specular})
   Roughness=LinearInterpolate(A=ScalarParameter() {parameterName=Roughness}, B=@n15.B, Alpha=ScalarParameter() {parameterName=Use Custom Roughness})
   Anisotropy
   EmissiveColor=@n18
   Opacity=Add(A=@n24, B=Multiply(A=@n18, B=@n25))
   OpacityMask=@n24
   Normal=@n34
   Tangent
   WorldPositionOffset
   SubsurfaceColor
   ClearCoat
   ClearCoatRoughness
   AmbientOcclusion
   Refraction
   CustomizedUVs_0
   CustomizedUVs_1
   CustomizedUVs_2
   CustomizedUVs_3
   CustomizedUVs_4
   CustomizedUVs_5
   CustomizedUVs_6
   CustomizedUVs_7
   PixelDepthOffset
   ShadingModel
   Displacement
  )
  MaterialFunctionCall(In=BlendMaterialAttributes(
    A=@n35
    B=MaterialFunctionCall!(Material=@n35) {materialFunction=&/MF_GumballFace}.Result
    Alpha=ScalarParameter() {parameterName=UseFace}
   )) {materialFunction=&MF_TechArtOverrides}
  Multiply(A=@n9)
  Multiply(
   A=@n10
   B=TextureSampleParameter2D() {parameterName=EmissiveMask, samplerType=SAMPLERTYPE_Masks, texture=&T_BlackMask}.RGB
  )
  Reroute(None=@n10)
  ScalarParameter() {parameterName=EmissiveStrength}
  n36 = TextureSampleParameter2D(UVs=TextureCoordinate()) {group=Mask Texture, parameterName=Diffuse Texture with Alpha Mask, sortPriority=10, texture=&T_BlacksRGB}
  StaticSwitchParameter(True=@n36.RGB, False=@n11.RGB) {parameterName="Use Diffuse Texture for Color [ignores alpha channel]"}
  n37 = TextureCoordinate()
  TextureSampleParameter2D(UVs=AppendVector(A=@n8, B=ChannelMaskParameter(None=@n37) {maskChannel=Green, defaultValue=<rgba 0 0 1 0>}.R)) {parameterName=Gradient, texture=/BRCosmetics/Characters/Player/Male/Medium/Bodies/M_MED_Gumball/Textures/T_M_Gumball_TF}
  Transform(None=@n34)
  VectorParameter() {parameterName=EmissiveColor}
 ]
}
