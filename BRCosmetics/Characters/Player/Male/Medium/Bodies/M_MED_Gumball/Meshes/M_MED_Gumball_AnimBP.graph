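# Post-process style AnimBP for the Gumball male cosmetic body, copying the attached parent's pose and adding deformation and secondary motion.
# During the Gumball emote, it blends a prop-flap sequence onto the machine handle and door bones, time-synced to the montage position.
# Three chained Control Rigs feed a spine-weighted door constraint, forearm bone tweaks, and AnimDynamics for zippers, canisters and pocket.
# The event graph updates emote weight and a crouch alpha that offsets the pocket.
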AnimBlueprint /BRCosmetics/Characters/Player/Male/Medium/Bodies/M_MED_Gumball/Meshes/M_MED_Gumball_AnimBP
target_skeleton=&/M_MED_Gumball_Skeleton
graphs [
 AnimGraph {nodes [
   n1 = Layered blend per bone(
    BasePose=Copy Pose From Mesh() {bUseAttachedParent=true}
    BlendPoses_0="Sequence Evaluator (by time) 'Emote_Gumball_CMM_Prop_Flap_JointsOnly'"!(ExplicitTime:float=Get Montage_Position() -> (Montage_Position:double)) -> (Pose:Pose Link Structure) {
     is=title AnimGraphNode_SequenceEvaluator
     sequence=/BRCosmetics/Animation/Game/MainPlayer/Emotes/Gumball/Emote_Gumball_CMM_Prop_Flap_JointsOnly
    }
    BlendWeights_0=Get Emote() -> (Emote:double)
   ) {layerSetup [{branchFilters [{boneName=dyn_machineHandle_01}, {boneName=dyn_machineDoor_01}]}]}
   n2 = Control Rig(Source=@n1) {
    alphaBoolBlend {blendOption=Linear}
    controlRigAssetReference {blueprintRigClass=&/M_MED_Gumball_Deform_CtrlRig_twists.M_MED_Gumball_Deform_CtrlRig_twists_C}
   }
   n3 = Control Rig!(
    Source=@n2
    Alpha
    KneeRot_fr_l_114deg:Vector
    KneeTrans_fr_l_114deg:Vector="0.258,-3.591,0.009"
    KneeScale_fr_l_114deg:Vector="1,1,1"
    KneeScale_fr_l_130deg:Vector="1,1,1"
    KneeRot_fr_l_130deg:Vector="0,0,-51.2"
    KneeTrans_fr_l_130deg:Vector="0.258,-3.591,0.009"
   ) -> (Pose) {
    alphaBoolBlend {blendOption=Linear}
    controlRigAssetReference {blueprintRigClass=&/M_MED_Gumball_Deform_CtrlRig_sculpters.M_MED_Gumball_Deform_CtrlRig_sculpters_C}
   }
   n4 = Control Rig!(Source=@n3, Alpha) -> (Pose) {
    alphaBoolBlend {blendOption=Linear}
    controlRigAssetReference {blueprintRigClass=&/M_MED_Gumball_CtrlRig.M_MED_Gumball_CtrlRig_C}
   }
   n5 = "Constraint - dyn_machineDoor_Base (RRR)"!(
    ComponentPose=Local To Component(LocalPose=@n4)
    Alpha=0.2
    ConstraintWeights_0=0.6
    ConstraintWeights_1:float=0.3
    ConstraintWeights_2:float=0.1
   ) -> (Pose) {
    alphaBoolBlend {blendOption=Linear}
    boneToModify {boneName=dyn_machineDoor_Base}
    constraintSetup [
     {transformType=Rotation, targetBone {boneName=spine_02}}
     {transformType=Rotation, targetBone {boneName=spine_03}}
     {transformType=Rotation, targetBone {boneName=spine_04}}
    ]
    constraintWeights [1,1,1]
   }
   n6 = "Transform (Modify) Bone - Bone: dyn_forearm_Lf_scale_01"(ComponentPose=@n5, Alpha=1, Translation="0,0,0", Rotation="0,0,0", Scale="0.01,0.01,0.01") {
    scaleMode=BMM_Additive
    translationMode=BMM_Additive
    alphaBoolBlend {blendOption=Linear}
    boneToModify {boneName=dyn_forearm_Lf_scale_01}
    scale=<0.01 0.01 0.01>
   }
   n7 = "Transform (Modify) Bone - Bone: dyn_forearm_Rt_scale_01"(ComponentPose=@n6, Alpha=1, Translation="0,0,0", Rotation="0,0,0", Scale="0.01,0.01,0.01") {
    scaleMode=BMM_Additive
    translationMode=BMM_Additive
    alphaBoolBlend {blendOption=Linear}
    boneToModify {boneName=dyn_forearm_Rt_scale_01}
    scale=<0.01 0.01 0.01>
   }
   n8 = Anim Dynamics - Bone: dyn_zipper_rt_01(ComponentPose=@n7, Alpha=1, ExternalForce="0,0,0") {
    angularDampingOverride=0.9
    angularSpringConstant=20000
    bAngularSpring=true
    bGravityOverrideInSimSpace=true
    bOverrideAngularDamping=true
    bUseGravityOverride=true
    numSolverIterationsPreUpdate=2
    simulationSpace=RootRelative
    alphaBoolBlend {blendOption=Linear}
    boundBone {boneName=dyn_zipper_rt_01}
    componentAppliedLinearAccClamp=<3000 3000 3000>
    componentLinearAccScale=<0.01 0.01 0.01>
    componentLinearVelScale=<0.01 0.01 0.01>
    gravityOverride {z=-980}
    physicsBodyDefinitions [{
      sphereCollisionRadius=0
      boundBone {boneName=dyn_zipper_rt_01}
      boxExtents=<5 5 4>
      constraintSetup {angularLimitsMax=<0 0 0>, angularLimitsMin=<0 -60 0>, angularTarget=<0 0 0>}
      localJointOffset=<0 0 2.5>
     }]
    relativeSpaceBone {boneName=calf_r}
   }
   n9 = Anim Dynamics - Bone: dyn_zipper_lf_01(ComponentPose=@n8, Alpha=1, ExternalForce="0,0,0") {
    angularDampingOverride=0.9
    angularSpringConstant=20000
    bAngularSpring=true
    bGravityOverrideInSimSpace=true
    bOverrideAngularDamping=true
    bUseGravityOverride=true
    numSolverIterationsPreUpdate=2
    simulationSpace=RootRelative
    alphaBoolBlend {blendOption=Linear}
    boundBone {boneName=dyn_zipper_lf_01}
    componentAppliedLinearAccClamp=<3000 3000 3000>
    componentLinearAccScale=<0.01 0.01 0.01>
    componentLinearVelScale=<0.01 0.01 0.01>
    gravityOverride {z=-980}
    physicsBodyDefinitions [{
      sphereCollisionRadius=0
      boundBone {boneName=dyn_zipper_lf_01}
      boxExtents=<5 5 4>
      constraintSetup {angularLimitsMax=<0 0 0>, angularLimitsMin=<0 -60 0>, angularTarget=<0 0 0>}
      localJointOffset=<0 0 -2.5>
     }]
    relativeSpaceBone {boneName=calf_l}
   }
   n10 = Anim Dynamics - Bone: dyn_cannisterA_Rt_01(ComponentPose=@n9, Alpha=0.5, ExternalForce="0,0,0") {
    angularDampingOverride=0.9
    angularSpringConstant=15000
    bAngularSpring=true
    bGravityOverrideInSimSpace=true
    bOverrideAngularDamping=true
    bUseGravityOverride=true
    numSolverIterationsPreUpdate=2
    simulationSpace=RootRelative
    alphaBoolBlend {blendOption=Linear}
    boundBone {boneName=dyn_cannisterA_Rt_01}
    componentAppliedLinearAccClamp=<3000 3000 3000>
    componentLinearAccScale=<0.01 0.01 0.01>
    componentLinearVelScale=<0.01 0.01 0.01>
    gravityOverride {z=-980}
    physicsBodyDefinitions [{
      sphereCollisionRadius=0
      boundBone {boneName=dyn_cannisterA_Rt_01}
      boxExtents=<10 10 8>
      constraintSetup {angularLimitsMax=<0 0 0>, angularLimitsMin=<0 -20 -20>, angularTarget=<0 0 0>}
      localJointOffset=<0 0 -2.5>
     }]
    relativeSpaceBone {boneName=spine_01}
   }
   n11 = Anim Dynamics - Bone: dyn_cannisterB_Rt_01(ComponentPose=@n10, Alpha=0.7, ExternalForce="0,0,0") {
    angularDampingOverride=0.9
    angularSpringConstant=18000
    bAngularSpring=true
    bGravityOverrideInSimSpace=true
    bOverrideAngularDamping=true
    bUseGravityOverride=true
    numSolverIterationsPreUpdate=2
    simulationSpace=RootRelative
    alphaBoolBlend {blendOption=Linear}
    boundBone {boneName=dyn_cannisterB_Rt_01}
    componentAppliedLinearAccClamp=<3000 3000 3000>
    componentLinearAccScale=<0.01 0.01 0.01>
    componentLinearVelScale=<0.01 0.01 0.01>
    gravityOverride {z=-980}
    physicsBodyDefinitions [{
      sphereCollisionRadius=0
      boundBone {boneName=dyn_cannisterB_Rt_01}
      boxExtents=<10 10 8>
      constraintSetup {angularLimitsMax=<0 0 0>, angularLimitsMin=<0 -20 -20>, angularTarget=<0 0 0>}
      localJointOffset=<0 0 -2.5>
     }]
    relativeSpaceBone {boneName=spine_01}
   }
   n12 = Anim Dynamics - Bone: dyn_pocket_Lf_01(ComponentPose=@n11, Alpha=0.6, ExternalForce="0,0,0") {
    angularDampingOverride=0.9
    angularSpringConstant=20000
    bAngularSpring=true
    bGravityOverrideInSimSpace=true
    bOverrideAngularDamping=true
    bUseGravityOverride=true
    numSolverIterationsPreUpdate=2
    simulationSpace=RootRelative
    alphaBoolBlend {blendOption=Linear}
    boundBone {boneName=dyn_pocket_Lf_01}
    componentAppliedLinearAccClamp=<3000 3000 3000>
    componentLinearAccScale=<0.01 0.01 0.01>
    componentLinearVelScale=<0.01 0.01 0.01>
    gravityOverride {z=-980}
    physicsBodyDefinitions [{
      sphereCollisionRadius=0
      boundBone {boneName=dyn_pocket_Lf_01}
      boxExtents=<10 10 8>
      constraintSetup {angularLimitsMax=<0 0 0>, angularLimitsMin=<0 -20 -20>, angularTarget=<0 0 0>}
      localJointOffset=<0 0 -2.5>
     }]
    relativeSpaceBone {boneName=spine_01}
   }
   n13 = "Transform (Modify) Bone - Bone: dyn_pocket_Lf_01"(ComponentPose=@n12, Alpha=Get isCrouchingAlpha() -> (isCrouchingAlpha:double), Translation="0,-3,0", Rotation="0,-20,0", Scale="1,1,1") {
    rotationMode=BMM_Additive
    translationMode=BMM_Additive
    alphaBoolBlend {blendOption=Linear}
    boneToModify {boneName=dyn_pocket_Lf_01}
    rotation {yaw=-20}
    translation {y=-3}
   }
   Output Pose(Result=Component To Local(ComponentPose=@n13)) {layerGroup=None}
  ]}
 EventGraph {nodes [
   Event BlueprintUpdateAnimation()
   > Get!(execute:Exec) -> (then:Exec, else:Exec, Skel_Mesh:Skeletal Mesh Component Object Reference)
   > Get!(execute) -> (then, else, Anim_Inst:Anim Instance Object Reference)
   > n1 = Sequence()
   > n2 = Branch(Condition=Get bIsCrouching())
   > Set isCrouchingAlpha(isCrouchingAlpha:double=1) -> (Output_Get:double)
   Set isCrouchingAlpha(execute=@n2.else, isCrouchingAlpha=0)
   n3 = Get Anim_Inst() -> (Anim_Inst:Anim Instance Object Reference)
   Set Emote(
    execute=@n1.then_1
    Emote:double=SelectFloat(
     A=1
     B=0
     bPickA=Montage_IsActive!(
      self:Anim Instance Object Reference=@n3
      Montage:Anim Montage Object Reference=/BRCosmetics/Animation/Game/MainPlayer/Emotes/Gumball/Emote_Gumball_CMM_M
     ) -> (ReturnValue:bool)
    )
   ) -> (Output_Get:double)
   > Set Montage_Position(Montage_Position:double=Montage_GetPosition!(
     self:Anim Instance Object Reference=@n3
     Montage:Anim Montage Object Reference=&Emote_Gumball_CMM_M
    ) -> (ReturnValue:float)) -> (Output_Get:double)
   Event BlueprintInitializeAnimation()
   > n4 = Set Skel_Mesh(Skel_Mesh=GetOwnerSkeletalMeshComponent())
   > Set Anim_Inst(Anim_Inst:Anim Instance Object Reference=GetAnimInstance(self=@n4)) -> (Output_Get:Anim Instance Object Reference)
  ]}
]
variables [
 Anim_Inst {type=AnimInstance, default=null}
 Emote {type=number}
 isCrouchingAlpha {type=number}
 Montage_Position {type=number}
 Skel_Mesh {type=SkeletalMeshComponent}
]
